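# Masked, two-sided lit material for liquid splash particle effects in a player emote, usable on Niagara sprites, ribbons and meshes.
# A radial flowmap, scaled by dynamic CenterWarp, distorts the tiled UVs sampling splash mask and normal textures.
# Particle alpha drives power-based erosion of the mask, while particle color optionally feeds an emissive adjustment function.
# Dynamic MeshWPOBend builds a Z offset weighted by cubed inverse V, and a parameter flattens the normal.
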Material /Game/Animation/Game/MainPlayer/Emotes/Faux/FX/Materials/M_Liquid_Splash
graph {
 properties {
  bUsedWithNiagaraMeshParticles=true
  bUsedWithNiagaraRibbons=true
  bUsedWithNiagaraSprites=true
  blendMode=BLEND_Masked
  pixelDepthOffsetMode=PDOM_Legacy
  twoSided=true
  usedShadingModels=Default Lit
 }
 nodes [
  n1 = ParticleColor()
  n2 = ScalarParameter() {parameterName=UseParticleColor}
  MaterialFunctionCall(Emissive=Multiply(A=LinearInterpolate(B=@n1.RGB, Alpha=@n2), B=ScalarParameter() {defaultValue=0.15, parameterName=EmissiveAdd}) {constB=0.15}) {materialFunction=&MF_FX_EmissiveAdjustments}
  OneMinus(None=Divide(A=Time() {bOverride_Period=true, period=2})) {desc=debug erosion}
  n3 = VectorParameter() {parameterName=Tile, defaultValue=<rgba 1 1 0 0>}
  n4 = MaterialFunctionCall(Texture Scale=Constant() {r=1.4}) {materialFunction=&ScaleUVsByCenter}
  n5 = Subtract(A=ComponentMask(None=TextureSample(UVs=@n4.UVs) {samplerType=SAMPLERTYPE_LinearColor, texture=/Game/Environments/Hexmap/HexClouds/T_RadialFlowmap01}.RGB) {g=true, r=true}) {constB=0.5}
  n6 = DynamicParameter() {parameterIndex=1, defaultValue=<rgba 0 0 1 1>, paramNames [CenterWarp,MeshWPOBend,Param3,Param4]}
  n7 = Multiply(A=AppendVector(A=@n3.R, B=@n3.R), B=Add(A=TextureCoordinate(), B=Multiply(A=@n5, B=@n6.CenterWarp)))
  n8 = TextureSampleParameter2D(UVs=@n7) {parameterName=Mask, texture=/Game/Effects/Fort_Effects/Textures/Water/T_water_splash_04}
  n9 = ChannelMaskParameter(None=@n8.RGB) {maskChannel=Blue, parameterName=ErodeMask, defaultValue=<rgba 0 0 1 0>}
  n10 = LinearInterpolate(Alpha=@n1.A) {constA=2, constB=-0.1}
  n11 = Multiply(A=Multiply(A=@n1.A, B=Power(Base=@n9.R, Exp=@n10)), B=ChannelMaskParameter(None=@n8.RGBA) {parameterName=OpacityMask}.R)
  Reroute(None=@n11)
  n12 = Constant3Vector() {constant=<rgba 0 0 1 0>}
  n13 = OneMinus(None=ComponentMask(None=TextureCoordinate()) {g=true})
  Reroute(None=Multiply(A=Multiply(A=@n12, B=@n6.MeshWPOBend), B=Power(Base=@n13) {constExponent=3}))
  Reroute(None=LinearInterpolate(
    A=TextureSampleParameter2D(UVs=@n7) {parameterName=Normal, samplerType=SAMPLERTYPE_Normal, texture=/Game/Effects/Fort_Effects/Textures/Water/T_water_splash_04_n}.RGB
    B=Constant3Vector() {constant=<rgba 0 0 1 0>}
    Alpha=ScalarParameter() {parameterName=Flatten Normal}
   ))
  Reroute(None=@n1.RGB)
  n14 = ScalarParameter() {defaultValue=1, parameterName=Translucency_Exp}
  Saturate(None=Multiply(A=Power(Base=@n11, Exp=@n14), B=ScalarParameter() {defaultValue=1, parameterName=Translucency_Hardness}))
  ScalarParameter() {defaultValue=0.2, parameterName=Roughness}
  ScalarParameter() {parameterName=Metallic}
  ScalarParameter() {defaultValue=0.5, parameterName=Specular}
 ]
}
